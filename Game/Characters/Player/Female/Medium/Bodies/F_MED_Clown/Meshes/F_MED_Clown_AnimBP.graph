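# Animation Blueprint for the female medium Clown player body part, adding secondary motion to its flowers and bandana.
# It copies the pose from the owner's skeletal mesh, cached on initialize.
# Chained Anim Dynamics nodes then simulate spring-driven wrist flowers, right forearm flower and right arm bandana in actor space.
# Each update, if the owning FortPlayerPawn has the JumpBoots AnimTracker tag, it sets the part material's "Hide Element 03" to 1, never resetting it.
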AnimBlueprint /Game/Characters/Player/Female/Medium/Bodies/F_MED_Clown/Meshes/F_MED_Clown_AnimBP
target_skeleton=&/F_MED_Clown_Skeleton
graphs [
 AnimGraph {nodes [
   n1 = Get MeshToCopy() -> (MeshToCopy:Skeletal Mesh Component Object Reference)
   n2 = Anim Dynamics - Bone: dyn_wrist_flower_l(ComponentPose=Local To Component(LocalPose=Copy Pose From Mesh(SourceMeshComponent=@n1)), Alpha=1, ExternalForce="0,0,0") {
    alpha=0.9
    angularSpringConstant=1500
    bAngularSpring=true
    bGravityOverrideInSimSpace=true
    gravityScale=5
    numSolverIterationsPostUpdate=4
    numSolverIterationsPreUpdate=16
    simulationSpace=Actor
    alphaBoolBlend {blendOption=Linear}
    boundBone {boneName=dyn_wrist_flower_l}
    physicsBodyDefinitions [{
      sphereCollisionRadius=0
      boundBone {boneName=dyn_wrist_flower_l}
      boxExtents=<25 10 10>
      constraintSetup {angularTargetAxis=AxisZ, twistAxis=AxisY, angularLimitsMax=<5 10 15>, angularLimitsMin=<-5 -10 -15>, angularTarget=<0 0 1>}
      localJointOffset=<0 0 0>
     }]
   }
   n3 = Anim Dynamics - Bone: dyn_wrist_flower_r(ComponentPose=@n2, Alpha=1, ExternalForce="0,0,0") {
    alpha=0.9
    angularSpringConstant=1500
    bAngularSpring=true
    bGravityOverrideInSimSpace=true
    gravityScale=0.5
    numSolverIterationsPostUpdate=4
    numSolverIterationsPreUpdate=16
    simulationSpace=Actor
    alphaBoolBlend {blendOption=Linear}
    boundBone {boneName=dyn_wrist_flower_r}
    physicsBodyDefinitions [{
      sphereCollisionRadius=0
      boundBone {boneName=dyn_wrist_flower_r}
      boxExtents=<25 10 10>
      constraintSetup {angularTargetAxis=AxisZ, twistAxis=AxisY, angularLimitsMax=<5 10 15>, angularLimitsMin=<-5 -10 -15>, angularTarget=<0 0 1>}
      localJointOffset=<0 0 0>
     }]
   }
   n4 = Anim Dynamics - Bone: dyn_forearm_flower_r(ComponentPose=@n3, Alpha=1, ExternalForce="0,0,0") {
    alpha=0.9
    angularSpringConstant=1500
    bAngularSpring=true
    bGravityOverrideInSimSpace=true
    gravityScale=0.5
    numSolverIterationsPostUpdate=4
    numSolverIterationsPreUpdate=16
    simulationSpace=Actor
    alphaBoolBlend {blendOption=Linear}
    boundBone {boneName=dyn_forearm_flower_r}
    physicsBodyDefinitions [{
      sphereCollisionRadius=0
      boundBone {boneName=dyn_forearm_flower_r}
      boxExtents=<25 10 10>
      constraintSetup {angularTargetAxis=AxisZ, twistAxis=AxisY, angularLimitsMax=<5 5 15>, angularLimitsMin=<-5 -5 -15>, angularTarget=<0 0 1>}
      localJointOffset=<0 0 0>
     }]
   }
   n5 = Anim Dynamics - Bone: dyn_arm_r_bandana(ComponentPose=@n4, Alpha=0.75, ExternalForce="0,0,0") {
    alpha=0.9
    angularSpringConstant=1500
    bAngularSpring=true
    bGravityOverrideInSimSpace=true
    numSolverIterationsPostUpdate=4
    numSolverIterationsPreUpdate=16
    simulationSpace=Actor
    alphaBoolBlend {blendOption=Linear}
    boundBone {boneName=dyn_arm_r_bandana}
    physicsBodyDefinitions [{
      sphereCollisionRadius=0
      boundBone {boneName=dyn_arm_r_bandana}
      boxExtents=<15 5 5>
      constraintSetup {angularTargetAxis=AxisZ, twistAxis=AxisY, angularLimitsMax=<15 0 15>, angularLimitsMin=<-15 0 -15>, angularTarget=<0 0 1>}
      localJointOffset=<0 0 0>
     }]
   }
   Output Pose(Result=Component To Local(ComponentPose=@n5)) {layerGroup=None}
  ]}
 EventGraph {nodes [
   Event BlueprintUpdateAnimation()
   > n1 = Cast To FortPlayerPawn(Object=GetOwningActor())
   > Is Valid(InputObject=@n1)
   > n2 = HasMatchingGameplayTag!(execute:Exec, self:Object Reference=@n1, TagToCheck:Gameplay Tag Structure=(TagName="Item.Gadget.JumpBoots.AnimTracker")) -> (then:Exec, ReturnValue:bool) {is=title K2Node_Message}
   > Branch(Condition=@n2)
   > n3 = Get Dynamic Material Instance(SkeletalMeshComponent=GetPartSkeletalMeshComponent())
   > SetScalarParameterValue(self=@n3, ParameterName=Hide Element 03, Value=1)
   Event BlueprintInitializeAnimation()
   > Set MeshToCopy(MeshToCopy:Skeletal Mesh Component Object Reference=GetOwnerSkeletalMeshComponent()) -> (Output_Get:Skeletal Mesh Component Object Reference)
  ]}
]
variables [MeshToCopy {type=SkeletalMeshComponent}]
